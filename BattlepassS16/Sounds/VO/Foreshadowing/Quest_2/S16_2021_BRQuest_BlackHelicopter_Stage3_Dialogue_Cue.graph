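# Sound cue for Stage 3 quest dialogue in the Season 16 Battle Pass Black Helicopter foreshadowing quest.
# A Dialogue Player plays the stage's dialogue wave in the S16QuestVO speaker context and feeds a Looping node at the root, so the line repeats.
# The cue uses the narrator VO sound class, a foreshadowing quest attenuation asset and a boosted volume.
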SoundCue /BattlepassS16/Sounds/VO/Foreshadowing/Quest_2/S16_2021_BRQuest_BlackHelicopter_Stage3_Dialogue_Cue
root=Looping_0
nodes [
 {
  id=DialoguePlayer_0
  properties {dialogueWaveParameter {
    dialogueWave=/BattlepassS16/Sounds/VO/Foreshadowing/Quest_2/Waves/S16_2021_BRQuest_BlackHelicopter_Stage3_Dialogue
    context {speaker=/BattlepassS16/Sounds/VO/Foreshadowing/S16QuestVO, targets []}
   }}
 }
 {id=Looping_0, children [DialoguePlayer_0]}
]
properties {
 attenuationSettings=/BattlepassS16/Sounds/VO/Foreshadowing/S17_Foreshadowing_Quest_VO_Att
 maxDistance=5400
 soundClassObject=/Game/Sounds/FortSoundClassesAndModes/Fort_VO_Narrator
 volumeMultiplier=1.5
}
